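; BPI_NonParticipant is a Blueprint interface for non-participant AI NPCs in Athena, covering their voice sounds.
; It declares AttemptPlayVocalReinforcementSound, which takes a gameplay tag container of metadata, and PlayVoiceSoundByTag, which takes a single sound bank tag.
; ShouldUseAlertSound returns a boolean for whether an alert sound is used.
; The interface holds only signatures, with no logic of its own.
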(fn AttemptPlayVocalReinforcementSound
  (params (MetadataTagContainer :type Gameplay%20Tag%20Container%20Structure)))

(fn PlayVoiceSoundByTag
  (params (SoundBankTag :type Gameplay%20Tag%20Structure)))

(fn ShouldUseAlertSound
  (return false))
