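; Base bot controller Blueprint for Athena NPCs that adjusts the controlled pawn when patrolling starts and stops.
; Patrol start applies a patrolling GameplayEffect, stores its handle, then saves the movement component's rotation rate and swaps in a patrolling rate.
; Patrol stop removes the effect and restores the saved rate.
; Removal by handle falls back to removal by the Phoebe.AIState.Patrolling tag, printing a warning if both fail.
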(event OnStartPatrollingEvent(FortAthenaNpcPatrolling)
  (CallFunction|StartPatrolling))

(event OnStopPatrollingEvent(FortAthenaNpcPatrolling)
  (CallFunction|StopPatrolling))

(fn AddPatrollingGE
  (bind _playerbotpawn (Variables|FortAthenaAIBotController|GetPlayerBotPawn))
  (bind _returnvalue (Ability|GetAbilitySystemComponent _playerbotpawn))
  (Utilities|IsValid _playerbotpawn
    (:Is%20Valid
      (Utilities|IsValid _returnvalue
        (:Is%20Valid
          (bind _returnvalue_2 (GameplayEffects|ApplyGameplayEffectToTarget :self _returnvalue :GameplayEffectClass (Variables|Default|GetGE_Patrolling) :Target _returnvalue :Level 0.0))
          (Variables|Default|SetGE_PatrollingSpeedHandle _returnvalue_2)
          (return))))))

(fn RemovePatrollingGE
  (bind _playerbotpawn (Variables|FortAthenaAIBotController|GetPlayerBotPawn))
  (bind _returnvalue (Ability|GetAbilitySystemComponent _playerbotpawn))
  (Utilities|IsValid _playerbotpawn
    (:Is%20Valid
      (Utilities|IsValid _returnvalue
        (:Is%20Valid
          (bind _returnvalue_2 (GameplayEffects|RemoveActiveGameplayEffect _returnvalue (Variables|Default|GetGE_PatrollingSpeedHandle) -1))
          (if (not _returnvalue_2)
            (bind _returnvalue_3 (GameplayEffects|RemoveActiveEffectswithTags _returnvalue "(GameplayTags=((TagName=\"Phoebe.AIState.Patrolling\")))"))
            (if (not (> _returnvalue_3 0))
              (Development|PrintString "Trying to remove a Patrolling GE from either the Handle or by Tag and it failed on both occasions. Should investigate.")))
          (return))))))

(fn StartPatrolling
  (bind _playerbotpawn (Variables|FortAthenaAIBotController|GetPlayerBotPawn))
  (bind _charactermovement (Variables|/Script/Engine.Character|GetCharacterMovement _playerbotpawn))
  (CallFunction|AddPatrollingGE)
  (Utilities|IsValid _playerbotpawn
    (:Is%20Valid
      (Variables|Default|SetSavedRotationRate (Variables|/Script/Engine.CharacterMovementComponent|GetRotationRate _charactermovement))
      (Variables|/Script/Engine.CharacterMovementComponent|SetRotationRate (Variables|Default|GetPatrollingRotationRate) _charactermovement))))

(fn StopPatrolling
  (bind _playerbotpawn (Variables|FortAthenaAIBotController|GetPlayerBotPawn))
  (CallFunction|RemovePatrollingGE)
  (Utilities|IsValid _playerbotpawn
    (:Is%20Valid
      (Variables|/Script/Engine.CharacterMovementComponent|SetRotationRate (Variables|Default|GetSavedRotationRate) (Variables|/Script/Engine.Character|GetCharacterMovement _playerbotpawn)))))

(fn ConstructionScript
  (|CallParentFunction))
