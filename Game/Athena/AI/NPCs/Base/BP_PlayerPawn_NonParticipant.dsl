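; Base pawn Blueprint for non-participant NPCs such as Henchmen, handling alertness, voice, animation, hit reactions, hiring and disguise removal.
; Controller events feed AlertLevel and target, whose RepNotifies drive "!"/"?" cues, data-tracker state, stress effects, voice lines and target-yaw tracking.
; Voice also arrives by replicated sound-bank tag or multicast.
; Patrol and override anim layers relink through anim input events; conversion grants tags and ability sets.
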(event EventBeginPlay
  ;[1 1 1] Begin Play
  (|CallParentFunction)
  (Utilities|FlowControl|Sequence
    (:then_0
      (if (Variables|Default|GetUseAlertness)
        (CallFunction|ActivateAlertDBNOCleanUp)))
    (:then_1
      (CallFunction|DataTrackerSetup true)
      (CallFunction|SoundLibrarySetup true))
    (:then_2
      (if (Variables|Default|GetUseFullBodyHitReact)
        (Ability|EquipFortAbilitySet self (Variables|Default|GetHitReactAbilitySet) self) ; Give Full Body Hit React Ability
        (CallFunction|InitializeAnimInstanceSettings)))
    (:then_3
      (Networking|SwitchHasAuthority
        (:Authority
          (CallFunction|ListenforVOEvent))
        (:Remote
          (if (Variables|VO|GetUseVO)
            (if (Variables|VO|GetUseIdleVO)
              (bind _returnvalue (Utilities|Time|SetTimerbyEvent (delegate Custom|TriggerIdleVO) (Math|Random|RandomFloatinRange (ScalableFloat|GetValueatLevel (Variables|VO|GetMinIdleVODelay)) (ScalableFloat|GetValueatLevel (Variables|VO|GetMaxIdleVODelay))) true))
              (Variables|Default|SetIdleVOTimerHandle _returnvalue))))))
    (:then_4
      (Default|BindEventtoOnEnteredVehicleDispatcher (EventDispatchers|CreateEvent :!event Custom|OnEnteredVehicleEvent))
      (Default|BindEventtoOnExitedVehicleDispatcher (EventDispatchers|CreateEvent :!event Custom|OnExitedVehicleEvent))
      (EventDispatchers|AssignOnDied (EventDispatchers|CreateEvent :!event Custom|OnDied_Event)))))

(event Custom|TriggerIdleVO
  ;[1 1 1] Begin Play
  (CallFunction|PlayVoiceComponentSoundbyTag (select (Utilities|Enum|Equal(Enum) (Variables|Default|GetAlertLevel) "Unaware") "(TagName=\"Pawn.Athena.NPC.Sound.Voice.Idle\")" "(TagName=\"Pawn.Athena.NPC.Sound.Voice.FullAlertIdle\")")))

(event Custom|ListenForVOEvent
  ;[1 1 1] Listen for VO Event: This will try to trigger a GC that will trigger a sound from the Sound Bank on the Client
  (bind (_asynctaskproxy _payload) (Ability|Async|WaitGameplayEventToActor 0 (Variables|SpeechBubblesVO|GetTagEventVOTrigger))
    (:EventReceived
      (if (> (Utilities|Array|Length (GameplayTags|BreakGameplayTagContainer (.TargetTags (Utilities|Struct|BreakGameplayEventData _payload)))) 0)
        (Variables|SpeechBubblesVO|SetVOSoundBankTag (Utilities|Array|Get(acopy) (GameplayTags|BreakGameplayTagContainer (.TargetTags (Utilities|Struct|BreakGameplayEventData _payload)))))
        (Utilities|FlowControl|Delay 1.0)
        (Variables|SpeechBubblesVO|SetVOSoundBankTag)))))

(event Custom|OnAlertLevelChangedEvent
  (params (BotController :type Fort%20Athena%20AIBot%20Controller%20Object%20Reference) (OldAlertLevel :type EAlertLevel%20Enum :default "Unaware") (NewAlertLevel :type EAlertLevel%20Enum :default "Unaware"))
  ;[1 1 1] Stealth Meter Update Event
  (CallFunction|AlertLevelChanged OldAlertLevel NewAlertLevel))

(event EventEndPlay
  (params (EndPlayReason :type EEndPlayReason%20Enum :default "Destroyed"))
  (|CallParentFunction EndPlayReason)
  (CallFunction|DataTrackerSetup false)
  (CallFunction|SoundLibrarySetup false)
  (Default|UnbindPawnAISpawnedEvent (Pawn|GetController)))

(event Custom|OnDied_Event
  (params (DamagedActor :type Actor%20Object%20Reference) (Damage :type Float%20%28single-precision%29) (InstigatedBy :type Controller%20Object%20Reference) (DamageCauser :type Actor%20Object%20Reference) (HitLocation :type Vector) (FHitComponent :type Primitive%20Component%20Object%20Reference) (BoneName :type Name) (Momentum :type Vector))
  ;[1 1 1] Stealth Meter Update Event
  (CallFunction|UnbindAlertLevelEvent (Pawn|GetController)))

(event EventOnWeaponEquipped
  (params (NewWeapon :type Fort%20Weapon%20Object%20Reference) (PrevWeapon :type Fort%20Weapon%20Object%20Reference))
  ;[1 1 1] When Pawn Changes Weapon, validate what type of Ammo we should drop.
  (Utilities|FlowControl|Sequence
    (:then_0
      (if (Variables|Default|GetUseDropExtraLootOnDeath)
        (Networking|SwitchHasAuthority
          (:Authority
            (CallFunction|SetupLootTierNamewithWeapon)))))))

(event Custom|OnEnteredVehicleEvent
  ;[1 1 1] Stealth Meter Update Event
  (CallFunction|UpdateAlertStateVehicleCheck))

(event Custom|OnExitedVehicleEvent
  ;[1 1 1] Stealth Meter Update Event
  (CallFunction|UpdateAlertStateVehicleCheck))

(event OnComponentHit(CapsuleComponent)
  (params (HitComponent :type Primitive%20Component%20Object%20Reference) (OtherActor :type Actor%20Object%20Reference) (OtherComp :type Primitive%20Component%20Object%20Reference) (NormalImpulse :type Vector :default "0, 0, 0") (Hit :type Hit%20Result%20Structure%20%28by%20ref%29))
  ;[1 1 1] If a disguised player bumps into a Henchman, remove the disguise.
  (if (Utilities|NotEqual(Object) OtherActor self)
    (bind _asplayer_pawn_athena (Utilities|Casting|CastToPlayerPawn_Athena OtherActor)) ; Only care about Pawns.
    (if (Team|OnSameTeam _asplayer_pawn_athena self)
      _
      (elif (GameplayTags|HasMatchingGameplayTag _asplayer_pawn_athena (Variables|Default|GetTC_DisguiseTag))
        (GameplayEffects|ApplyGameplayEffectToTarget :self (Ability|GetAbilitySystemComponent _asplayer_pawn_athena) :GameplayEffectClass (Variables|Default|GetGE_RemoveDisguise) :Target (Ability|GetAbilitySystemComponent _asplayer_pawn_athena) :Level 0.0) ; Remove the disguise.
      ))))

(event Custom|ActivateAlertDBNOCleanUp
  ;[1 1 1] Clear up "!" and "?" VFX immediately on DBNO
  (EventDispatchers|AssignOnDBNOStateChanged self (delegate Custom|OnDBNOStateChanged_Event)))

(event Custom|OnDBNOStateChanged_Event
  (params (FortPawn :type Fort%20Pawn%20Object%20Reference) (bInIsDBNO :type Boolean))
  ;[1 1 1] Clear up "!" and "?" VFX immediately on DBNO
  ;;[1 1 1] Enable full movement simulation when going into DBNO to fix pops with throw/drop/interrogation animations
  (CallFunction|SetShouldSkipMovementFullSimulation (not bInIsDBNO))
  (if bInIsDBNO
    (CallFunction|SetAlertedStateGameplayCue false)
    (CallFunction|PlayVoiceComponentSoundbyTag "(TagName=\"Pawn.Athena.NPC.Sound.Voice.DBNO\")")
    (GameplayEffects|ApplyGameplayEffectToSelf :self (Ability|GetAbilitySystemComponent self) :GameplayEffectClass (Variables|Default|GetGE_NPC_Status_Stressed_Infinite) :Level 0.0 :!pinned true) ; DBNO = Infinite Stress\nPropagate to other Henchmen.
    (elif (and (Utilities|Select :Index (Variables|Default|GetAlertLevel) :Unaware true :Suspicious false :LostTarget false :Threatened false :Count false) (not (FortPawn|IsDead self)))
      (Networking|SwitchHasAuthority
        (:Authority
          (GameplayEffects|ApplyGameplayEffectToSelf :self (Ability|GetAbilitySystemComponent self) :GameplayEffectClass (Variables|Default|GetGE_NPC_Status_Stressed_Cooldown) :Level 0.0 :!pinned true) ; Recover from DBNO = Cooldown stress that bleeds off.\nWill run around in patrol for a short time if he has no target.
        )))))

(event EventOnCharacterCustomizationCompleted
  (params (Pawn :type Fort%20Player%20Pawn%20Object%20Reference))
  ;[1 1 1] Anim Setup
  (|CallParentFunction Pawn)
  (CallFunction|ReLinkAnimLayer)
  (FortPawn|AssignOnWeaponAttachmentChangedDelegate (delegate Custom|OnWeaponAttachmentChangedDelegate_Event_0)))

(event Vehicle|EventOnExitedVehicle
  ;[1 1 1] On Vehicle Exit
  (|CallParentFunction)
  (CallFunction|ReLinkAnimLayer))

(event Custom|OnWeaponAttachmentChangedDelegate_Event_0
  (params (NewWeapon :type Fort%20Weapon%20Object%20Reference) (PrevWeapon :type Fort%20Weapon%20Object%20Reference))
  ;[1 1 1] Anim Setup
  (CallFunction|UpdatePatrolAnimSet NewWeapon))

(event GameplayCue|GameplayCue.Player.Interrogation.Voice.PickedUp
  (params (EventType :type EGameplayCueEvent%20Enum :default "OnActive") (Parameters :type Gameplay%20Cue%20Parameters%20Structure%20%28by%20ref%29))
  ;[1 1 1] Interrogation GCs from Interrogation system
  (Networking|CanExecuteCosmeticEvents
    (:True
      (CallFunction|PlayVoiceComponentSoundbyTag "(TagName=\"Pawn.Athena.NPC.Sound.Voice.Interrogate\")"))))

(event EventOnDeathPlayEffects
  (params (Damage :type Float%20%28single-precision%29) (DamageTags :type Gameplay%20Tag%20Container%20Structure%20%28by%20ref%29) (Momentum :type Vector :default "0, 0, 0") (HitInfo :type Hit%20Result%20Structure%20%28by%20ref%29) (InstigatedBy :type Fort%20Pawn%20Object%20Reference) (DamageCauser :type Actor%20Object%20Reference) (EffectContext :type Gameplay%20Effect%20Context%20Handle%20Structure))
  ;[1 1 1] On Death Effects
  (|CallParentFunction Damage DamageTags Momentum HitInfo InstigatedBy DamageCauser EffectContext)
  (bind _returnvalue (DeathCause|ToDeathCause DamageTags))
  (Default|CallOnDeath _returnvalue)
  (CallFunction|PlayVoiceComponentSoundbyTag "(TagName=\"Pawn.Athena.NPC.Sound.Voice.DBNO\")")
  (Utilities|Time|ClearandInvalidateTimerbyHandle (Variables|Default|GetIdleVOTimerHandle)))

(event Custom|SetChallenger
  (params (Challenger :type Fort%20Player%20Pawn%20Object%20Reference))
  ;[1 1 1] Challenger Tagging
  (bind _output_get (Variables|Duel|SetChallenger Challenger))
  (EventDispatchers|BindEventtoOnDied _output_get (delegate Custom|ChallengerDied))
  (bind _returnvalue (GameplayEffects|ApplyGameplayEffectToSelf :self (Ability|GetAbilitySystemComponent self) :GameplayEffectClass (Variables|Default|GetGE_DuelTag) :Level 0.0))
  (Variables|Duel|SetGEDuelHandle _returnvalue))

(event Custom|ChallengerDied
  (params (DamagedActor :type Actor%20Object%20Reference) (Damage :type Float%20%28single-precision%29) (InstigatedBy :type Controller%20Object%20Reference) (DamageCauser :type Actor%20Object%20Reference) (HitLocation :type Vector) (FHitComponent :type Primitive%20Component%20Object%20Reference) (BoneName :type Name) (Momentum :type Vector))
  ;[1 1 1] Challenger Tagging
  (GameplayEffects|RemoveActiveGameplayEffect (Ability|GetAbilitySystemComponent self) (Variables|Duel|GetGEDuelHandle) -1)
  (Variables|Duel|SetGEDuelHandle))

(event Custom|StartCalculatingTargetYawDegree
  ;[1 1 1] Calculate the Degree of Difference on the Yaw between this Pawn and his Target (if he has one)
  (bind _returnvalue (Utilities|Time|SetTimerbyEvent (EventDispatchers|CreateEvent :!event CalculateYawDegreeWithTarget) 0.1 true))
  (Variables|InternalLogic|SetHNDL_CalculateYawTargetDegreeTimer _returnvalue))

(event Custom|LinkPatrolAnimLayer
  ;[1 1 1] Link Patrol AnimLayer / Link Override AnimLayer
  (if (not (ScalableFloat|GetValueasBool (Variables|Animation|GetUseMotionMatchingForPatrols)))
    (Components|SkeletalMesh|AnimationBlueprintLinking|UnlinkAnimClassLayers (Variables|Character|GetMesh) (Variables|Default|GetOverrideLayerAnimBP))
    (Components|SkeletalMesh|AnimationBlueprintLinking|LinkAnimClassLayers (Variables|Character|GetMesh) (Variables|Default|GetPatrolLayerAnimBP))
    (CallFunction|UpdatePatrolAnimSet (Variables|Weapon|GetCurrentWeapon)))
  (CallFunction|InitializeAnimInstanceSettings))

(event Custom|LinkOverrideAnimLayer
  ;[1 1 1] Link Patrol AnimLayer / Link Override AnimLayer
  (if (not (ScalableFloat|GetValueasBool (Variables|Animation|GetUseMotionMatchingForPatrols)))
    (Components|SkeletalMesh|AnimationBlueprintLinking|UnlinkAnimClassLayers (Variables|Character|GetMesh) (Variables|Default|GetPatrolLayerAnimBP))
    (Components|SkeletalMesh|AnimationBlueprintLinking|LinkAnimClassLayers (Variables|Character|GetMesh) (Variables|Default|GetOverrideLayerAnimBP))))

(event Custom|ReLinkAnimLayer
  ;[1 1 1] Link Patrol AnimLayer / Link Override AnimLayer
  (if (Utilities|IsValidClass (Variables|Default|GetOverrideLayerAnimBP))
    (CallFunction|LinkOverrideAnimLayer)
    (else
      (CallFunction|LinkPatrolAnimLayer))))

(event Animation|EventOnAnimInputEvent
  (params (AnimInputEvent :type Fort%20Anim%20Input%20Event%20Object%20Reference))
  ;[1 1 1] Process incoming AnimInputEvents
  (if (Class|FortAnimInputEvent|IsNamed AnimInputEvent "aie.npc.OverridePatrolLayer")
    (bind _asfort_anim_input_event_anim_class (Utilities|Casting|CastToFortAnimInputEvent_AnimClass AnimInputEvent))
    (Variables|Default|SetOverrideLayerAnimBP (Variables|/Script/FortniteGame.FortAnimInputEvent_AnimClass|GetValue _asfort_anim_input_event_anim_class))
    (CallFunction|LinkOverrideAnimLayer)
    (elif (Class|FortAnimInputEvent|IsNamed AnimInputEvent "aie.npc.RestorePatrolLayer")
      (CallFunction|LinkPatrolAnimLayer)
      (Variables|Default|SetOverrideLayerAnimBP 0) ; Clear the Override Layer
    )))

(event BPI_NonParticipant|EventPlayVoiceSoundByTag
  (params (SoundBankTag :type Gameplay%20Tag%20Structure))
  (Networking|CanExecuteCosmeticEvents
    (:True
      (bind _fromtrueinput (Utilities|FlowControl|MergeExecution)
        (:Then
          (CallFunction|PlayVoiceComponentSoundbyTag (select _fromtrueinput SoundBankTag (stale Custom|On%20Multicast%20Play%20Voice%20Sound%20By%20Tag SoundBankTag))))))
    (:False
      (CallFunction|OnMulticastPlayVoiceSoundbyTag SoundBankTag))))

(event Custom|On%20Multicast%20Play%20Voice%20Sound%20By%20Tag
  (params (SoundBankTag :type Gameplay%20Tag%20Structure))
  (bind _fromtrueinput (Utilities|FlowControl|MergeExecution)
    (:Then
      (CallFunction|PlayVoiceComponentSoundbyTag (select _fromtrueinput (stale BPI_NonParticipant|EventPlayVoiceSoundByTag SoundBankTag) SoundBankTag)))))

(event OnConvertedEvent(ConvertComponent)
  (params (InstigatorPawn :type Fort%20Pawn%20Object%20Reference) (ConvertedPawn :type Fort%20Pawn%20Object%20Reference))
  ;[1 1 1] On Converted - grant a gameplay tag to the converter denoting they've hired an NPC
  (Conversion|HandlePawnConverted InstigatorPawn ConvertedPawn)
  (CallFunction|EventonHiredAudio InstigatorPawn))

(event OnUnconvertedEvent(ConvertComponent)
  (params (UnconvertedPawn :type Fort%20Pawn%20Object%20Reference) (UnconvertReason :type EUnconvertReason%20Enum :default "StealByOtherPlayer"))
  ;[1 1 1] On Converted - grant a gameplay tag to the converter denoting they've hired an NPC
  (Conversion|HandlePawnUnconverted UnconvertedPawn))

(event Custom|Event%20OnHiredAudio
  (params (Instigator :type Fort%20Pawn%20Object%20Reference))
  ;[1 1 1] Play sound 2D for player hiring NPC, 3D for other players in team
  (CallFunction|PlayVoiceComponentSoundbyTag "(TagName=\"Pawn.Athena.NPC.Sound.Voice.Hire\")")
  (if (Team|OnLocalPlayersTeam Instigator)
    (CallFunction|PlayNpcStinger false (Utilities|Struct|MakeSimpleSoundSettings (Variables|Default|GetSlStingerTag) "/Game/Sounds/Attenuation/Fort_Grenade_Shock_Att.Fort_Grenade_Shock_Att")) ; Connect NOT for switching logic
  ))

(event EventReceiveControllerChanged
  (params (OldController :type Controller%20Object%20Reference) (NewController :type Controller%20Object%20Reference))
  ;[1 1 1] On Controller Changed
  (|CallParentFunction OldController NewController)
  (Utilities|FlowControl|Sequence
    (:then_0
      (Default|UnbindPawnAISpawnedEvent OldController)
      (Default|UnbindEscalateGaugeEvents OldController)
      (CallFunction|UnbindAlertLevelEvent OldController)
      (Default|UnbindCurrentTargetChangedEvent OldController))
    (:then_1
      (Default|BindPawnAISpawnedEvent NewController)
      (Default|BindEscalateGaugeEvents NewController)
      (CallFunction|BindAlertLevelEvent NewController)
      (Default|BindCurrentTargetChangedEvent NewController))))

(event Custom|OnPawnAISpawned
  (params (BotController :type AIController%20Object%20Reference) (BotPawn :type Fort%20Pawn%20Object%20Reference))
  (Default|BindEscalateGaugeEvents BotController))

(event Custom|OnEscalateGaugeStarted
  (params (EscalatingController :type Controller%20Object%20Reference) (EscalationTarget :type Fort%20Pawn%20Object%20Reference) (ActiveEscalateTags :type Gameplay%20Tag%20Container%20Structure))
  (bind _tag_escalationspeech (Variables|Default|GetTag_EscalationSpeech))
  ;[1 1 1] Escalate Gauge Events
  (if (and (Variables|Default|GetUseEscalationSpeechBubbles) (GameplayTags|IsGameplayTagValid _tag_escalationspeech))
    (Class|BPFLAICommonFunctions|PlaySpeechBubble self _tag_escalationspeech true true)))

(event Custom|OnEscalateGaugeFinished
  (params (EscalatingController :type Controller%20Object%20Reference) (EscalationTarget :type Fort%20Pawn%20Object%20Reference) (ActiveEscalateTags :type Gameplay%20Tag%20Container%20Structure)))

(event Custom|AiTargetChanged
  (params (OldTarget :type Actor%20Object%20Reference) (NewTarget :type Actor%20Object%20Reference))
  ;[1 1 1] AI Target Changed
  (Variables|Default|SetNewTargetActor NewTarget))

(event Custom|DisableDBNOCarry
  (params (Delay :type Float%20%28double-precision%29))
  ; Needs to be called on Client
  (Utilities|FlowControl|Delay Delay)
  (bind _asfort_ability_system_component (Utilities|Casting|CastToFortAbilitySystemComponent (Ability|GetAbilitySystemComponent self)))
  (Tags|BlockAbilitiesWithTags _asfort_ability_system_component (Variables|Default|GetDisableDBNOCarryTag)))

(fn AlertLevelChanged
  (params (OldAlertLevel :type EAlertLevel%20Enum :default "Unaware") (InputPin :type EAlertLevel%20Enum :default "Unaware"))
  (Variables|Default|SetAlertLevel InputPin))

(fn BindAlertLevelEvent
  (params (Controller :type Controller%20Object%20Reference))
  (Utilities|IsValid Controller
    (:Is%20Valid
      (bind _asbp_phoebe_controller_non_participant (Utilities|Casting|CastToBP_PhoebeController_NonParticipant Controller)
        (:then
          (AlertLevel|BindEventtoOnAlertLevelChangedEvent _asbp_phoebe_controller_non_participant (EventDispatchers|CreateEvent :!event OnAlertLEvelCHangedEvent)))
        (:CastFailed
          (bind _asfort_athena_aibot_controller (Utilities|Casting|CastToFortAthenaAIBotController Controller))
          (if (GameplayTags|HasMatchingGameplayTag "(TagName=\"Athena.AI.ShowAlertState\")")
            (AlertLevel|BindEventtoOnAlertLevelChangedEvent _asfort_athena_aibot_controller (EventDispatchers|CreateEvent :!event OnAlertLEvelCHangedEvent))))))))

(fn BindCurrentTargetChangedEvent
  (params (Controller :type Controller%20Object%20Reference))
  (Utilities|IsValid Controller
    (:Is%20Valid
      (if (Variables|Default|GetUseChangedTargetReplication)
        (bind _asfort_athena_aibot_controller (Utilities|Casting|CastToFortAthenaAIBotController Controller))
        (Perception|BindEventtoOnCurrentTargetChangedEvent _asfort_athena_aibot_controller (EventDispatchers|CreateEvent :!event AiTargetChanged))))))

(fn BindEscalateGaugeEvents
  (params (Controller :type Controller%20Object%20Reference))
  (bind _returnvalue (Actor|GetComponentbyClass Controller "/Script/FortniteAI.FortControllerComponent_AIEscalate"))
  (Utilities|IsValid Controller
    (:Is%20Valid
      (Utilities|IsValid _returnvalue
        (:Is%20Valid
          (EventDispatchers|BindEventtoEscalateGaugeStartedDelegate _returnvalue (EventDispatchers|CreateEvent :!event OnEscalateGaugeStarted))
          (EventDispatchers|BindEventtoEscalateGaugeFinishedDelegate _returnvalue (EventDispatchers|CreateEvent :!event OnEscalateGaugeFinished)))))))

(fn BindPawnAISpawnedEvent
  (params (Controller :type Controller%20Object%20Reference))
  (Utilities|IsValid Controller
    (:Is%20Valid
      (bind _asfort_athena_aibot_controller (Utilities|Casting|CastToFortAthenaAIBotController Controller))
      (Utilities|IsValid _asfort_athena_aibot_controller
        (:Is%20Valid
          (Spawning|BindEventtoOnPawnAISpawnedEvent _asfort_athena_aibot_controller (EventDispatchers|CreateEvent :!event OnPawnAISpawned)))))))

(fn CalculateYawDegreeWithTarget
  (bind _newtargetactor (Variables|Default|GetNewTargetActor))
  (Utilities|IsValid _newtargetactor
    (:Is%20Valid
      (Variables|InternalLogic|SetTargetYawDegrees (select (< (Math|Vector|DotProduct (Transformation|GetActorRightVector self) (Math|Vector|Normalize (- (Transformation|GetActorLocation _newtargetactor) (Transformation|GetActorLocation self)))) 0) (Class|AbilityFunctions|GetAngleBetweenTwoVectors (Math|Vector|Normalize (- (Transformation|GetActorLocation _newtargetactor) (Transformation|GetActorLocation self))) (Transformation|GetActorForwardVector self)) (* (Class|AbilityFunctions|GetAngleBetweenTwoVectors (Math|Vector|Normalize (- (Transformation|GetActorLocation _newtargetactor) (Transformation|GetActorLocation self))) (Transformation|GetActorForwardVector self)) -1.0))))
    (:Is%20Not%20Valid
      (Variables|InternalLogic|SetTargetYawDegrees 0.0))))

(fn ClientOnAlertLevelChanged
  (params (OldAlertLevel :type EAlertLevel%20Enum :default "Unaware") (NewAlertLevel :type EAlertLevel%20Enum :default "Unaware"))
  (Default|CallOnAlertLevelChanged OldAlertLevel NewAlertLevel))

(fn ClientOnTargetChanged
  (params (NewTarget :type Actor%20Object%20Reference))
  (Default|CallOnTargetChanged NewTarget))

(fn DataTrackerSetup
  (params (Is%20Registering :type Boolean))
  (bind _self self)
  (bind _datatrackercomponentmanager (Variables|Default|GetDataTrackerComponentManager))
  (Networking|CanExecuteCosmeticEvents
    (:True
      (Variables|Default|GetDataTrackerComponentManager
        (:then
          _)
        (:else
          (bind _returnvalue (FortGameplayDataTrackerComponent|GetFortGameplayDataTrackerComponentManager))
          (bind _output_get (Variables|Default|SetDataTrackerComponentManager _returnvalue))))
      (Utilities|IsValid _output_get
        (:Is%20Valid
          (if (Variables|GetIs%20Registering)
            (Class|FortGameplayDataTrackerComponentManager|RegisterTrackedStateActor _datatrackercomponentmanager _self (Variables|AlertLevelTags|GetAlertLevelUnawareTag))
            (else
              (Class|FortGameplayDataTrackerComponentManager|UnRegisterTrackedStateActor _datatrackercomponentmanager _self))))))))

(fn GetCurrentFullbodyHitReactionMontage
  (bind _currentpatrolanimasset (Variables|Default|GetCurrentPatrolAnimAsset))
  (bind _animset (Variables|/Script/FortniteGame.FortPatrolAnimAsset|GetAnimSet _currentpatrolanimasset))
  (bind (_patrolidlepose _patrolidleadditive _additivehitreactmontage _fullbodyhitreactmontage) (Utilities|Struct|BreakFortAnimInputPatrolAnimSet _animset))
  (if (Math|Boolean|ANDBoolean (not (Variables|Character|GetbIsCrouched)) (Pawn|Components|CharacterMovement|IsWalking (Variables|Character|GetCharacterMovement)) (Utilities|IsValid _currentpatrolanimasset :!class /Script/Engine.KismetSystemLibrary))
    (Utilities|IsValid _fullbodyhitreactmontage
      (:Is%20Valid
        (return _fullbodyhitreactmontage))
      (:Is%20Not%20Valid
        _)))
  (return (Variables|Default|GetDefaultFullbodyHitReactionMontage)))

(fn GetIconPlacement
  (params (SelfActor :type Actor%20Object%20Reference) (ViewingActor :type Actor%20Object%20Reference))
  (bind (_outlocation _outextents) (|CallParentFunction SelfActor ViewingActor))
  (bind _mesh (Variables|Character|GetMesh))
  (Utilities|IsValid _mesh
    (:Is%20Valid
      (return (Math|Vector|SelectVector _outlocation (Transformation|GetSocketLocation _mesh "FX_Spine") (not (Transformation|DoesSocketExist _mesh "FX_Spine"))) _outextents))
    (:Is%20Not%20Valid
      (return _outlocation _outextents))))

(fn HandleAlertStateUpdates
  (bind _self self)
  (bind _returnvalue (Ability|GetAbilitySystemComponent _self))
  (bind _self_2 self)
  (bind _returnvalue_2 (AI|GetBlackboard _self_2))
  (bind _returnvalue_3 (Utilities|Name|MakeLiteralName "AIEvaluator_Global_TargetActor"))
  (bind _returnvalue_4 (AI|Components|Blackboard|GetValueasObject _returnvalue_2 _returnvalue_3))
  (bind _self_3 self)
  (bind _returnvalue_5 (Ability|GetAbilitySystemComponent _self_3))
  (bind _alertlevel (Variables|Default|GetAlertLevel))
  (bind _oldalertlevel (Variables|Default|GetOldAlertLevel))
  (bind _oldalertlevel_2 (Variables|Default|GetOldAlertLevel))
  (bind _alertlevel_2 (Variables|Default|GetAlertLevel))
  (if (Variables|Default|GetUseAlertness)
    (Utilities|FlowControl|Sequence
      (:then_0
        (if (Math|Boolean|ORBoolean (Variables|Health|GetbIsDBNO) (GameplayTags|HasMatchingGameplayTag self "(TagName=\"Gameplay.Status.DBNO\")") (FortPawn|IsDead self) (Variables|Default|GetbIsInVehicleThatSupportsNoAlertState))
          (CallFunction|SetAlertedStateGameplayCue false)
          (Utilities|IsValid _returnvalue
            (:Is%20Valid
              (GameplayEffects|ApplyGameplayEffectToSelf :self _returnvalue :GameplayEffectClass (Variables|Default|GetGE_NPC_Status_Stressed_Infinite) :Level 0.0 :!pinned true) ; INFINITE STRESS\n\nLet Patrol system know we were recently alerted, so we patrol faster for a while. "Stressed"
            ))
          (else
            (CallFunction|SetDataTrackerActorStatefromAlertLevel (Variables|Default|GetAlertLevel))
            (if (Utilities|Enum|Equal(Enum) _alertlevel_2 "Unaware")
              (CallFunction|SetAlertedStateGameplayCue false)
              (if (Utilities|Select :Index (Variables|Default|GetOldAlertLevel) :Unaware false :Suspicious true :LostTarget true :Threatened true :Count false)
                (GameplayEffects|ApplyGameplayEffectToSelf :self (Ability|GetAbilitySystemComponent self) :GameplayEffectClass (Variables|Default|GetGE_NPC_Status_Stressed_Cooldown) :Level 0.0 :!pinned true) ; STRESS THAT COOLS DOWN\n\nLet Patrol system know we were recently alerted, so we patrol faster for a while. "Stressed"
              )
              (elif (Utilities|Enum|Equal(Enum) _alertlevel_2 "Suspicious")
                (CallFunction|SetAlertedStateGameplayCue true)
                (if (Utilities|Select :Index _oldalertlevel_2 :Unaware false :Suspicious false :LostTarget true :Threatened true :Count false)
                  _
                  (elif (Utilities|Select :Index _oldalertlevel_2 :Unaware true :Suspicious false :LostTarget false :Threatened false :Count false)
                    (Ability|SendGameplayEventtoActor self "(TagName=\"Event.Animation.Alerted\")" :!pinned true) ; HACK TEST FOR HIGHLIGHTING ESCALATING ALERT LEVELS WITH TRIGGERED ANIMATIONS
                  ))
                (Utilities|IsValid _returnvalue_5
                  (:Is%20Valid
                    (GameplayEffects|ApplyGameplayEffectToSelf :self _returnvalue_5 :GameplayEffectClass (Variables|Default|GetGE_NPC_Status_Stressed_Infinite) :Level 0.0 :!pinned true) ; INFINITE STRESS\n\nLet Patrol system know we were recently alerted, so we patrol faster for a while. "Stressed"
                  )) ; Validate to Remove Errors when closing PIE
                (elif (or (Utilities|Enum|Equal(Enum) _alertlevel_2 "LostTarget") (Utilities|Enum|Equal(Enum) _alertlevel_2 "Threatened"))
                  (CallFunction|SetAlertedStateGameplayCue true true)
                  (if (Utilities|Select :Index (Variables|Default|GetOldAlertLevel) :Unaware true :Suspicious true :LostTarget false :Threatened false :Count false)
                    (Networking|SwitchHasAuthority
                      (:Authority
                        (Utilities|IsValid _returnvalue_2
                          (:Is%20Valid
                            (Utilities|IsValid _returnvalue_4
                              (:Is%20Valid
                                (bind _asactor (Utilities|Casting|CastToActor _returnvalue_4)
                                  (:then
                                    (if (> (Class|AbilityFunctions|GetDistanceBetweenActors _asactor self) (Variables|Animation|GetMinDistanceToTargetToPlayCombatAlertAnimation))
                                      (Ability|SendGameplayEventtoActor self "(TagName=\"Event.Animation.Combat\")" :!pinned true) ; HACK TEST FOR HIGHLIGHTING ESCALATING ALERT LEVELS WITH TRIGGERED ANIMATIONS
                                      (else
                                        _)))
                                  (:CastFailed
                                    _)))
                              (:Is%20Not%20Valid
                                _)))
                          (:Is%20Not%20Valid
                            _)))))))))))
      (:then_1
        (Networking|CanExecuteCosmeticEvents
          (:True
            (CallFunction|ClientonAlertLevelChanged _oldalertlevel _alertlevel)
            (CallFunction|PlayAlertLevelVO _oldalertlevel _alertlevel)))))))

(fn HandlePawnConverted
  (params (InstigatorPawn :type Fort%20Pawn%20Object%20Reference) (ConvertedPawn :type Fort%20Pawn%20Object%20Reference))
  (bind _gastoapplyonconvertedpawn (Variables|Conversion|GetGASToApplyOnConvertedPawn))
  (Variables|Conversion|SetPawnConverter InstigatorPawn)
  (bind _returnvalue (GameplayEffects|ApplyGameplayEffectToSelf :self (Ability|GetFortAbilitySystemComponent InstigatorPawn) :GameplayEffectClass (Variables|Conversion|GetGE_Quest_Converted_HireNPC) :Level 0.0))
  (Variables|Conversion|SetGE_Quest_Converted_Handle _returnvalue)
  (Ability|GameplayEffect|AddLooseGameplayTags (Variables|Conversion|GetPawnConverter) (Variables|Conversion|GetTagsToApplyOnConvertingPawn))
  (Ability|GameplayEffect|AddLooseGameplayTags ConvertedPawn (Variables|Conversion|GetTagsToApplyOnConvertedPawn))
  (Networking|SwitchHasAuthority
    (:Authority
      (Utilities|IsValid _gastoapplyonconvertedpawn
        (:Is%20Valid
          (bind _returnvalue_2 (Ability|EquipFortAbilitySet ConvertedPawn _gastoapplyonconvertedpawn ConvertedPawn))
          (Variables|Conversion|SetHNDL_EquippedAbilitySet _returnvalue_2))))))

(fn HandlePawnUnconverted
  (params (UnconvertedPawn :type Fort%20Pawn%20Object%20Reference))
  (bind _pawnconverter (Variables|Conversion|GetPawnConverter))
  (bind _returnvalue (Ability|GetAbilitySystemComponent _pawnconverter))
  (Utilities|IsValid _pawnconverter
    (:Is%20Valid
      (Utilities|IsValid _returnvalue
        (:Is%20Valid
          (GameplayEffects|RemoveActiveGameplayEffect _returnvalue (Variables|Conversion|GetGE_Quest_Converted_Handle) -1))
        (:Is%20Not%20Valid
          _))
      (if (Variables|Conversion|GetRemoveTagsOnUnconverted)
        (Ability|GameplayEffect|RemoveLooseGameplayTags (Variables|Conversion|GetPawnConverter) (Variables|Conversion|GetTagsToApplyOnConvertingPawn))
        (Ability|GameplayEffect|RemoveLooseGameplayTags UnconvertedPawn (Variables|Conversion|GetTagsToApplyOnConvertedPawn))
        (else
          _))
      (if (and (Variables|Conversion|GetRemoveGASOnUnconverted) (Utilities|IsValid (Variables|Conversion|GetGASToApplyOnConvertedPawn) :!class /Script/Engine.KismetSystemLibrary))
        (Ability|UnequipFortAbilitySet (Variables|Conversion|GetHNDL_EquippedAbilitySet))))
    (:Is%20Not%20Valid
      _)))

(fn InitializeAnimInstanceSettings
  (bind _asfort_player_anim_instance (Utilities|Casting|CastToFortPlayerAnimInstance (Components|SkeletalMesh|GetAnimInstance (Variables|Character|GetMesh))))
  (Variables|/Script/FortniteGame.FortPlayerAnimInstance|SetbShouldProcessFullAnimUpdate (Variables|Default|GetbShouldDoFullAnimationUpdate) _asfort_player_anim_instance)
  (Variables|/Script/FortniteGame.FortAnimInstance|SetbShouldPredictLanding false _asfort_player_anim_instance))

(fn OnRep_AlertLevel
  (CallFunction|HandleAlertStateUpdates)
  (Variables|Default|SetOldAlertLevel (Variables|Default|GetAlertLevel)))

(fn OnRep_bIsInVehicleThatSupportsNoAlertState
  (CallFunction|HandleAlertStateUpdates))

(fn OnRep_NewTargetActor
  (bind _hndl_calculateyawtargetdegreetimer (Variables|InternalLogic|GetHNDL_CalculateYawTargetDegreeTimer))
  (Networking|SwitchHasAuthority
    (:Remote
      (Default|ClientonTargetChanged (Variables|Default|GetNewTargetActor))
      (if (Variables|Default|GetCalculateTargetYawDegrees)
        (Utilities|IsValid (Variables|Default|GetNewTargetActor)
          (:Is%20Valid
            (if (not (Utilities|Time|IsValidTimerHandle _hndl_calculateyawtargetdegreetimer))
              (CallFunction|StartCalculatingTargetYawDegree)))
          (:Is%20Not%20Valid
            (Utilities|Time|ClearandInvalidateTimerbyHandle _hndl_calculateyawtargetdegreetimer)))))))

(fn OnRep_VOSoundBankTag
  (bind _vosoundbanktag (Variables|SpeechBubblesVO|GetVOSoundBankTag))
  (Networking|SwitchHasAuthority
    (:Remote
      (if (GameplayTags|IsGameplayTagValid _vosoundbanktag)
        (BPI_NonParticipant|PlayVoiceSoundbyTag _vosoundbanktag)))))

(fn PlayAlertLevelVO
  (params (OldAlertLevel :type EAlertLevel%20Enum :default "Unaware") (NewAlertLevel :type EAlertLevel%20Enum :default "Unaware"))
  (bind _oldalertlevellocal (Variables|GetOldAlertLevelLocal))
  (if (not (Math|Boolean|ORBoolean (Variables|Health|GetbIsDBNO) (GameplayTags|HasMatchingGameplayTag self "(TagName=\"Gameplay.Status.DBNO\")") (FortPawn|IsDead self) (not (Variables|Default|GetUseAlertness))))
    (LocalVariables|SetOldAlertLevelLocal OldAlertLevel)
    (if (Utilities|Enum|Equal(Enum) NewAlertLevel "Unaware")
      (if (Utilities|Select :Index (Variables|GetOldAlertLevelLocal) :Unaware false :Suspicious false :LostTarget true :Threatened true :Count false)
        (CallFunction|PlayVoiceComponentSoundbyTag "(TagName=\"Pawn.Athena.NPC.Sound.Voice.LostTarget\")")
        (else
          (CallFunction|PlayVoiceComponentSoundbyTag "(TagName=\"Pawn.Athena.NPC.Sound.Voice.ReturnToIdle\")")))
      (elif (Utilities|Enum|Equal(Enum) NewAlertLevel "Suspicious")
        (if (Utilities|Select :Index _oldalertlevellocal :Unaware false :Suspicious false :LostTarget true :Threatened true :Count false)
          (CallFunction|PlayVoiceComponentSoundbyTag "(TagName=\"Pawn.Athena.NPC.Sound.Voice.LostTarget\")")
          (elif (Utilities|Select :Index _oldalertlevellocal :Unaware true :Suspicious false :LostTarget false :Threatened false :Count false)
            (CallFunction|PlayVoiceComponentSoundbyTag "(TagName=\"Pawn.Athena.NPC.Sound.Voice.Suspicious\")")))
        (elif (or (Utilities|Enum|Equal(Enum) NewAlertLevel "LostTarget") (Utilities|Enum|Equal(Enum) NewAlertLevel "Threatened"))
          (if (Utilities|Select :Index (Variables|GetOldAlertLevelLocal) :Unaware true :Suspicious true :LostTarget false :Threatened false :Count false)
            (CallFunction|PlayVoiceComponentSoundbyTag "(TagName=\"Pawn.Athena.NPC.Sound.Voice.FullAlert\")")))))))

(fn PlayDeathEffects
  (bind _success (UEFNCustomEffects|PlayCustomEffect "(TagName=\"NPC.CustomEffect.DeathEffect\")" self))
  (if _success
    (FortPawn|SetPawnVisibilityLocal self true true)
    (else
      (CallFunction|PlayVoiceComponentSoundbyTag "(TagName=\"Pawn.Athena.NPC.Sound.Voice.DBNO\")")
      (Utilities|Time|ClearandInvalidateTimerbyHandle (Variables|Default|GetIdleVOTimerHandle))
      (|CallParentFunction))))

(fn PlayNpcStinger
  (params (OverideContext :type Boolean) (SlSettings :type Simple%20Sound%20Settings%20Structure))
  (bind _playresult (LocalVariables|GetPlayResult))
  (bind (_eventname _bwasblocked _bsuccess _audiocomponents) (Utilities|Struct|BreakSoundLibraryPlayResult _playresult))
  (bind _soundlibrarycomponent (Variables|SoundLibrary|GetSoundLibraryComponent))
  ;[1 1 1] Saftey checks
  (Networking|CanExecuteCosmeticEvents
    (:True
      (if (Variables|Sound|GetbAllowStingers)
        (Utilities|IsValid _soundlibrarycomponent
          (:Is%20Valid
            (if OverideContext
              ;[1 1 1] Update sound settings prior to playback
              (bind _returnvalue (SoundLibrary|PlaySoundSimple (WorldSubsystems|GetSoundLibrarySubsystem) self SlSettings :!class /Script/SoundLibrary.SoundLibrarySubsystem))
              (LocalVariables|SetPlayResult _returnvalue)
              (else
                ;[1 1 1] simple play sound
                (bind _returnvalue_2 (SoundLibrary|PlayEvent _soundlibrarycomponent (Utilities|Struct|BreakSimpleSoundSettings SlSettings)))
                (LocalVariables|SetPlayResult _returnvalue_2)))
            ;[1 1 1] Send stinger to music system
            (if (Utilities|Array|IsValidIndex _audiocomponents)
              (Audio|AddEventbyComponent (GameInstanceSubsystems|GetMusicEventSubsystem) (Utilities|Array|Get(acopy) _audiocomponents) "(TagName=\"MusicEventSubsystem.Event.NPC\")" "(GameplayTags=)" 0 ""))))))))

(fn PlayVoiceComponentSoundByTag
  (params (SoundBankTag :type Gameplay%20Tag%20Structure))
  (bind _foundvoicecomponent (LocalVariables|GetFoundVoiceComponent))
  (bind _audio_component (Variables|/Game/Sounds/Fort_Audio_Blueprints/VoiceComponent/NPC_VoiceComponent.NPC_VoiceComponent_C|GetAudio%20Component _foundvoicecomponent))
  (if (Variables|VO|GetUseVO)
    (Networking|CanExecuteCosmeticEvents
      (:True
        (bind _foundvoicecomponent_2 (LocalVariables|GetFoundVoiceComponent)
          (:then
            _)
          (:else
            (LocalVariables|SetFoundVoiceComponent (Actor|GetComponentbyClass "/Game/Sounds/Fort_Audio_Blueprints/VoiceComponent/NPC_VoiceComponent.NPC_VoiceComponent_C"))))
        (Utilities|IsValid _foundvoicecomponent_2
          (:Is%20Valid
            (bind _success (Class|NPCVoiceComponent|PlayVoiceLine _foundvoicecomponent SoundBankTag (select (GameplayTags|DoesContainerMatchTagQuery (GameplayTags|MakeGameplayTagContainerfromTag SoundBankTag) (Variables|VO|GetVOPlayDetachedTagQuery)) 0 (Variables|Default|GetVOAudioComponent))))
            (Utilities|IsValid _audio_component
              (:Is%20Valid
                (Transformation|AttachComponentToComponent _audio_component (Variables|Character|GetMesh) "head" "SnapToTarget" "SnapToTarget" "SnapToTarget" true)
                (if _success
                  (return (Variables|Default|GetVOAudioComponent))
                  (else
                    (return 0)))))))))))

(fn SetAlertedStateGameplayCue
  (params (Alerted :type Boolean) (Combat :type Boolean))
  (bind _returnvalue (Math|Integer|SelectInt 1 0 Combat))
  (bind _returnvalue_2 (Ability|GameplayCue|MakeGameplayCueParameters :NormalizedMagnitude 0.0 :RawMagnitude 0.0 :Location "0,0,0" :Normal "0,0,0" :GameplayEffectLevel _returnvalue :AbilityLevel 1 :bReplicateLocationWhenUsingMinimalRepProxy false))
  (bind _self self)
  (bind _gameplaycuealertstate (Variables|Default|GetGameplayCueAlertState))
  (if Alerted
    (Projectile|AddGameplayCue%28Looping%29Local _self _gameplaycuealertstate _returnvalue_2 :!pinned true) ; "!" or "?" On
    (else
      (Projectile|RemoveGameplayCue%28Looping%29Local _self _gameplaycuealertstate _returnvalue_2 :!pinned true) ; "!" or "?" Off
    )))

(fn SetDataTrackerActorStateFromAlertLevel
  (params (AlertLevel :type EAlertLevel%20Enum :default "Unaware"))
  (Networking|CanExecuteCosmeticEvents
    (:True
      (bind _datatrackercomponentmanager (Variables|Default|GetDataTrackerComponentManager)
        (:then
          (Class|FortGameplayDataTrackerComponentManager|SetTrackedActorState _datatrackercomponentmanager self (Utilities|Select :Index AlertLevel :Unaware (Variables|AlertLevelTags|GetAlertLevelUnawareTag) :Suspicious (Variables|AlertLevelTags|GetAlertLevelAlertedTag) :LostTarget (Variables|AlertLevelTags|GetAlertLevelAlertedTag) :Threatened (Variables|AlertLevelTags|GetAlertLevelAggressiveTag))))
        (:else)))))

(fn SetNPCAnimSpeedScalar
  (params (SpeedScalar :type Float%20%28double-precision%29))
  (bind _mesh (Variables|Character|GetMesh))
  (bind _returnvalue (Components|SkeletalMesh|GetAnimInstance _mesh))
  (bind _returnvalue_2 (Animation|GetAnimationInterface))
  (Utilities|FlowControl|Sequence
    (:then_0
      (Utilities|IsValid _returnvalue
        (:Is%20Valid
          (bind _asfort_patrol_anim_layer (Utilities|Casting|CastToFortPatrolAnimLayer (Animation|LinkedAnimGraphs|GetLinkedAnimLayerInstancebyClass _returnvalue (Variables|Default|GetPatrolLayerAnimBP) false)))
          (Variables|/Script/FortniteGame.FortPatrolAnimLayer|SetPatrolAnimSpeedScalar SpeedScalar _asfort_patrol_anim_layer))))
    (:then_1
      (if (Utilities|IsValid _returnvalue_2 :!class /Script/Engine.KismetSystemLibrary)
        (Locomotion|SetLocomotionAnimSpeedScalar _returnvalue_2 SpeedScalar)))))

(fn SetupLootTierNameWithWeapon
  (bind _currentweapon (Variables|Weapon|GetCurrentWeapon))
  (bind _returnvalue (Class|FortWeapon|GetWeaponDataAmmoWorldItemDefinition _currentweapon))
  (Utilities|Casting|CastToB_Athena_Pickaxe_Generic _currentweapon
    (:CastFailed
      (Utilities|IsValid _returnvalue
        (:Is%20Valid
          (Variables|Loot|SetLootTierNameToSpawnWhenDead (select (GameplayTags|HasTag (GameplayTags|GetOwnedGameplayTags _returnvalue :!class /Script/GameplayTags.GameplayTagAssetInterface) "(TagName=\"supplies.ammo.shell\")" true) (Variables|Loot|GetLootTier_Shell) (select (GameplayTags|HasTag (GameplayTags|GetOwnedGameplayTags _returnvalue :!class /Script/GameplayTags.GameplayTagAssetInterface) "(TagName=\"supplies.ammo.bulletheavy\")" true) (Variables|Loot|GetLootTier_Heavy) (select (GameplayTags|HasTag (GameplayTags|GetOwnedGameplayTags _returnvalue :!class /Script/GameplayTags.GameplayTagAssetInterface) "(TagName=\"supplies.ammo.bulletlight\")" true) (Variables|Loot|GetLootTier_Light) (select (GameplayTags|HasTag (GameplayTags|GetOwnedGameplayTags _returnvalue :!class /Script/GameplayTags.GameplayTagAssetInterface) "(TagName=\"supplies.ammo.bulletmedium\")" true) (Variables|Loot|GetLootTier_Medium) (select (GameplayTags|HasTag (GameplayTags|GetOwnedGameplayTags _returnvalue :!class /Script/GameplayTags.GameplayTagAssetInterface) "(TagName=\"supplies.ammo.explosive\")" true) (Variables|Loot|GetLootTier_Rockets) (Variables|Loot|GetLootTier_Medium))))))))
        (:Is%20Not%20Valid
          (Variables|Loot|SetLootTierNameToSpawnWhenDead (Variables|Loot|GetLootTier_Medium))))
      (return))) ; If we switch to a Pickaxe, don't update the Ammo we drop.
)

(fn ShouldUseAlertSound
  (return (Variables|Default|GetUseAlertSound)))

(fn SoundLibrarySetup
  (params (IsAdding :type Boolean))
  (bind _returnvalue (WorldSubsystems|GetSoundLibrarySubsystem))
  (Utilities|IsValid _returnvalue
    (:Is%20Valid
      (if IsAdding
        (bind _pawnsoundlibrary (Variables|Sound|GetPawnSoundLibrary)
          (:then
            (SoundLibrary|AddLibrary _returnvalue self _pawnsoundlibrary :!class /Script/SoundLibrary.SoundLibrarySubsystem))
          (:else))
        (else
          (bind _pawnsoundlibrary_2 (Variables|Sound|GetPawnSoundLibrary)
            (:then
              (SoundLibrary|RemoveLibrary _returnvalue self _pawnsoundlibrary_2 :!class /Script/SoundLibrary.SoundLibrarySubsystem))
            (:else)))))))

(fn UnbindAlertLevelEvent
  (params (Controller :type Controller%20Object%20Reference))
  (Utilities|IsValid Controller
    (:Is%20Valid
      (bind _asfort_athena_aibot_controller (Utilities|Casting|CastToFortAthenaAIBotController Controller))
      (AlertLevel|UnbindEventfromOnAlertLevelChangedEvent _asfort_athena_aibot_controller (EventDispatchers|CreateEvent :!event OnAlertLEvelCHangedEvent)))))

(fn UnbindCurrentTargetChangedEvent
  (params (Controller :type Controller%20Object%20Reference))
  (Utilities|IsValid Controller
    (:Is%20Valid
      (bind _asfort_athena_aibot_controller (Utilities|Casting|CastToFortAthenaAIBotController Controller))
      (Perception|UnbindEventfromOnCurrentTargetChangedEvent _asfort_athena_aibot_controller (EventDispatchers|CreateEvent :!event AiTargetChanged)))))

(fn UnbindEscalateGaugeEvents
  (params (Controller :type Controller%20Object%20Reference))
  (bind _returnvalue (Actor|GetComponentbyClass Controller "/Script/FortniteAI.FortControllerComponent_AIEscalate"))
  (Utilities|IsValid Controller
    (:Is%20Valid
      (Utilities|IsValid _returnvalue
        (:Is%20Valid
          (EventDispatchers|UnbindEventfromEscalateGaugeStartedDelegate _returnvalue (EventDispatchers|CreateEvent :!event OnEscalateGaugeStarted))
          (EventDispatchers|UnbindEventfromEscalateGaugeFinishedDelegate _returnvalue (EventDispatchers|CreateEvent :!event OnEscalateGaugeFinished)))))))

(fn UnbindPawnAISpawnedEvent
  (params (Controller :type Controller%20Object%20Reference))
  (Utilities|IsValid Controller
    (:Is%20Valid
      (bind _asfort_athena_aibot_controller (Utilities|Casting|CastToFortAthenaAIBotController Controller))
      (Utilities|IsValid _asfort_athena_aibot_controller
        (:Is%20Valid
          (Spawning|UnbindEventfromOnPawnAISpawnedEvent _asfort_athena_aibot_controller (EventDispatchers|CreateEvent :!event OnPawnAISpawned)))))))

(fn UpdateAlertStateVehicleCheck
  (bind _self self)
  (bind _returnvalue (Vehicle|GetVehicleInterface _self))
  (if (Vehicle|IsinVehicle _self)
    (if (Utilities|IsValid (Utilities|ToObject(Interface) _returnvalue) :!class /Script/Engine.KismetSystemLibrary)
      (if (not (GameplayTags|HasTag (Class|FortVehicleInterface|GetVehicleTags _returnvalue) "(TagName=\"Athena.Vehicle.AI.ShowAlertState\")"))
        (Variables|Default|SetbIsInVehicleThatSupportsNoAlertState true)
        (else
          (Variables|Default|SetbIsInVehicleThatSupportsNoAlertState false)))
      (else
        (Variables|Default|SetbIsInVehicleThatSupportsNoAlertState false)))
    (else
      (Variables|Default|SetbIsInVehicleThatSupportsNoAlertState false))))

(fn UpdatePatrolAnimSet
  (params (NewWeapon :type Fort%20Weapon%20Object%20Reference))
  (bind _mesh (Variables|Character|GetMesh))
  (bind _returnvalue (Components|SkeletalMesh|GetAnimInstance _mesh))
  (bind _patrolanimsetpairs (Variables|Default|GetPatrolAnimSetPairs))
  (bind _foundanimasset (LocalVariables|GetFoundAnimAsset))
  (Utilities|IsValid _returnvalue
    (:Is%20Valid
      (bind _asfort_patrol_anim_layer (Utilities|Casting|CastToFortPatrolAnimLayer (Animation|LinkedAnimGraphs|GetLinkedAnimLayerInstancebyClass _returnvalue (Variables|Default|GetPatrolLayerAnimBP) false)))
      (Utilities|IsValid (Class|FortPatrolAnimLayer|GetPatrolAnimSetfromWeapon _asfort_patrol_anim_layer _patrolanimsetpairs NewWeapon)
        (:Is%20Valid
          (LocalVariables|SetFoundAnimAsset (Class|FortPatrolAnimLayer|GetPatrolAnimSetfromWeapon _asfort_patrol_anim_layer _patrolanimsetpairs NewWeapon)))
        (:Is%20Not%20Valid
          (LocalVariables|SetFoundAnimAsset (Variables|Default|GetFallbackPatrolAnimAsset)) ; Fallback on NoWeaponAnimSet when no equipped weapon or no valid anim set found
        ))
      (Class|FortPatrolAnimLayer|SetPatrolAnimAsset _asfort_patrol_anim_layer _foundanimasset)
      (Variables|Default|SetCurrentPatrolAnimAsset _foundanimasset))))

(fn ConstructionScript
  (bind _returnvalue (Actor|GetComponentsByClass "/Script/FortniteAI.FortAIComponent_Voice"))
  (|CallParentFunction)
  (if (not (Variables|VO|GetUseVO))
    (Components|DestroyComponent (Variables|Default|GetVOAudioComponent))
    (if (== (Utilities|Array|Length _returnvalue) 1)
      (Components|DestroyComponent (Utilities|Array|Get(acopy) _returnvalue)))))
