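; Base Blueprint for non-participant boss NPC pawns, handling their health bar, speech bubbles and death fall.
; The server writes health and shield into HealthInfo, whose RepNotify updates clients' status widget and boss name for nearby players, with temporary force-hide support.
; Alert changes, escalations and death trigger random tagged speech bubbles.
; Optionally, death line-traces to ground and teleports the body down along a timer-driven curve.
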(event EventBeginPlay
  (|CallParentFunction)
  (Components|SkinnedMesh|SetPhysicsAsset (Variables|Character|GetMesh) "/Game/Characters/Player/Male/Male_Avg_Base/Fortnite_M_Avg_Player_Physics.Fortnite_M_Avg_Player_Physics" true) ; Hack to make sure the Physic Asset fits the size of the AI
  (Networking|SwitchHasAuthority
    (:Authority
      (Health|BindEventtoOnPawnHealthChanged (EventDispatchers|CreateEvent :!event Custom|PawnHealthChanged))
      (Variables|Default|GetShowHealth
        (:then
          (Default|CallOnBossReadyforFinalFight false)
          (Utilities|Time|SetTimerbyEvent (delegate Custom|EnableHealthBarDelayedEvent) 1.0))
        (:else)))
    (:Remote
      (Variables|Default|GetShowHealth
        (:then
          (CallFunction|StartUpdatingHealthWidget))
        (:else)))))

(event Custom|EnableHealthBarDelayedEvent
  (Variables|Default|SetEnableHealthBar true))

(event Custom|PawnHealthChanged
  ; Pawn Health Changed
  (Variables|Default|SetHealthInfo (Utilities|Struct|MakeStructNPCHealthInfo (Math|Float|SafeDivide (FortPawn|GetHealth) (FortPawn|GetMaxHealth)) (Math|Float|SafeDivide (FortPawn|GetShield) (FortPawn|GetMaxShield)) (FortPawn|GetMaxHealth) (FortPawn|GetMaxShield))))

(event Collision|EventActorBeginOverlap
  (params (OtherActor :type Actor%20Object%20Reference))
  (|CallParentFunction OtherActor) ; This node is disabled and will not be called.\nDrag off pins to build functionality.
)

(event EventTick
  (params (DeltaSeconds :type Float%20%28single-precision%29))
  (|CallParentFunction DeltaSeconds) ; This node is disabled and will not be called.\nDrag off pins to build functionality.
)

(event EventOnDeathServer
  (params (Damage :type Float%20%28single-precision%29) (DamageTags :type Gameplay%20Tag%20Container%20Structure%20%28by%20ref%29) (Momentum :type Vector :default "0, 0, 0") (HitInfo :type Hit%20Result%20Structure%20%28by%20ref%29) (InstigatedBy :type Controller%20Object%20Reference) (DamageCauser :type Actor%20Object%20Reference) (EffectContext :type Gameplay%20Effect%20Context%20Handle%20Structure)))

(event Custom|StartUpdatingHealthWidget
  ;[1 1 1] Health Bar
  (bind _npcstatuswidget (Variables|Default|GetNPCStatusWidget)
    (:then
      (bind _asnpcstatus_widget_aiboss (Utilities|Casting|CastToNPCStatusWidget_AIBoss (UserInterface|GetUserWidgetObject _npcstatuswidget)))
      (Utilities|FlowControl|Sequence
        (:then_0
          (Components|Tick|SetComponentTickEnabled (Variables|Default|GetNPCStatusWidget) (Widget|IsVisible _asnpcstatus_widget_aiboss)))
        (:then_1
          (Default|BindEventtoVisibilityChanged _asnpcstatus_widget_aiboss (delegate Custom|WidgetVisibilityChanged)))))
    (:else)))

(event EventEndPlay
  (params (EndPlayReason :type EEndPlayReason%20Enum :default "Destroyed"))
  (if (Variables|Default|GetShowHealth)
    (CallFunction|StopUpdatingHealthWidget)))

(event Custom|ForceHideHealthBar
  (params (ForceStayHiddenDuration :type Float%20%28double-precision%29))
  ;[1 1 1] Allow to Force Hide Health Bar and not wait until damage timer has expired - useful for cases such as when a Boss Cloaks
  (CallFunction|NPCDisableHealthBars)
  (if (> ForceStayHiddenDuration 0)
    (bind _returnvalue (Utilities|Time|SetTimerbyEvent (delegate Custom|RemoveForceHideHealthBar) ForceStayHiddenDuration))
    (Variables|InternalData|SetHNDL_ForceHiddenTimer _returnvalue)))

(event EventOnDeathPlayEffects
  (params (Damage :type Float%20%28single-precision%29) (DamageTags :type Gameplay%20Tag%20Container%20Structure%20%28by%20ref%29) (Momentum :type Vector :default "0, 0, 0") (HitInfo :type Hit%20Result%20Structure%20%28by%20ref%29) (InstigatedBy :type Fort%20Pawn%20Object%20Reference) (DamageCauser :type Actor%20Object%20Reference) (EffectContext :type Gameplay%20Effect%20Context%20Handle%20Structure))
  ;[1 1 1] On Death Client
  (|CallParentFunction Damage DamageTags Momentum HitInfo InstigatedBy DamageCauser EffectContext)
  (if (Variables|Default|GetShowHealth)
    (CallFunction|NPCSetupHealthBaronDeath)
    (else
      _))
  (if (Variables|Default|GetShouldSnapToGroundOnDeath)
    (CallFunction|TrySnaptoGroundonDeath)))

(event Custom|RemoveForceHideHealthBar
  ;[1 1 1] Allow to Force Hide Health Bar and not wait until damage timer has expired - useful for cases such as when a Boss Cloaks
  (CallFunction|NPCEnableHealthBars)
  (Utilities|Time|ClearandInvalidateTimerbyHandle (Variables|InternalData|GetHNDL_ForceHiddenTimer)))

(event Custom|WidgetVisibilityChanged
  (params (bVisible :type Boolean))
  ;[1 1 1] Health Bar
  (Components|Tick|SetComponentTickEnabled (Variables|Default|GetNPCStatusWidget) bVisible)
  (if bVisible
    (bind _returnvalue (Utilities|Time|SetTimerbyEvent (EventDispatchers|CreateEvent :!event Custom|UpdateWidgetOffset) 0.1 true))
    (Variables|Default|SetUpdateOffsetTimerHandle _returnvalue)
    (else
      (Utilities|Time|ClearandInvalidateTimerbyHandle (Variables|Default|GetUpdateOffsetTimerHandle)))))

(event Custom|BossReadyToEngage
  ;[1 1 1] Used for the Progress Bar UI Health
  (Default|CallOnBossReadyforFinalFight true))

(event Custom|StopUpdatingHealthWidget
  ;[1 1 1] Health Bar
  (bind _npcstatuswidget (Variables|Default|GetNPCStatusWidget)
    (:then
      (bind _asnpcstatus_widget_aiboss (Utilities|Casting|CastToNPCStatusWidget_AIBoss (UserInterface|GetUserWidgetObject _npcstatuswidget)))
      (Default|UnbindEventfromVisibilityChanged _asnpcstatus_widget_aiboss (delegate Custom|WidgetVisibilityChanged))
      (Utilities|Time|ClearandInvalidateTimerbyHandle (Variables|Default|GetUpdateOffsetTimerHandle)))
    (:else)))

(event Custom|NPC%20SetupHealthBarOnDeath
  (bind (_healthpercent_2_7edada384c7908ea98debd9958a249de _shieldpercent_5_b79c63b64aacf275a1621a9315964dcd _maxhealth_6_7d7bef7841b13fb96d2578989cf135bb _maxshield_9_f923e4444e2e476c56fb0e9d9d2f7846) (Utilities|Struct|BreakStructNPCHealthInfo (Variables|Default|GetHealthInfo)))
  ;[1 1 1] Health Bar Handling on Death
  (if (> (Variables|Default|GetNPC%20HealthBarHideTimeAfterDeath) 0.0)
    (bind _npcstatuswidget (Variables|Default|GetNPCStatusWidget)
      (:then
        (bind _asnpcstatus_widget_aiboss (Utilities|Casting|CastToNPCStatusWidget_AIBoss (UserInterface|GetUserWidgetObject _npcstatuswidget)))
        (Class|NPCStatusWidgetAIBoss|EventUpdateStatus _asnpcstatus_widget_aiboss 0.0 0.0 _maxhealth_6_7d7bef7841b13fb96d2578989cf135bb _maxshield_9_f923e4444e2e476c56fb0e9d9d2f7846 false false)
        (Utilities|FlowControl|Delay (Variables|Default|GetNPC%20HealthBarHideTimeAfterDeath)))
      (:else))
    (else
      _))
  (bind _npcstatuswidget_2 (Variables|Default|GetNPCStatusWidget)
    (:then
      (Development|SetHiddeninGame _npcstatuswidget_2 true false :!pinned true) ; Hide health bar (probably after animation)
    )
    (:else)))

(event EventOnAlertLevelChangedEvent
  (params (BotController :type Fort%20Athena%20AIBot%20Controller%20Object%20Reference) (OldAlertLevel :type EAlertLevel%20Enum :default "Unaware") (NewAlertLevel :type EAlertLevel%20Enum :default "Unaware"))
  ; Handle Alert Level Changes for Possible Speech Bubbles
  (|CallParentFunction BotController OldAlertLevel NewAlertLevel)
  (SpeechBubbles|HandleTriggerSpeechBubbleonAlertLevelChanged OldAlertLevel NewAlertLevel))

(event EventOnEscalateGaugeStarted
  (params (EscalatingController :type Controller%20Object%20Reference) (EscalationTarget :type Fort%20Pawn%20Object%20Reference) (ActiveEscalateTags :type Gameplay%20Tag%20Container%20Structure))
  ; Handle Alert Level Changes for Possible Speech Bubbles
  (|CallParentFunction EscalatingController EscalationTarget ActiveEscalateTags)
  (SpeechBubbles|HandleEscalationSpeechBubble ActiveEscalateTags))

(event Custom|NPCDisableHealthBars
  ;[1 1 1] Debug For Captures
  (Variables|Default|SetNPC%20StatusWidgetHealthBarDisabled true)
  (bind _npcstatuswidget (Variables|Default|GetNPCStatusWidget)
    (:then
      (bind _asnpcstatus_widget_aiboss (Utilities|Casting|CastToNPCStatusWidget_AIBoss (UserInterface|GetUserWidgetObject _npcstatuswidget)))
      (Widget|SetVisibility _asnpcstatus_widget_aiboss "Hidden"))
    (:else)))

(event Custom|NPCEnableHealthBars
  ;[1 1 1] Debug For Captures
  (Variables|Default|SetNPC%20StatusWidgetHealthBarDisabled false))

(event Custom|HandleCombatSpeechBubbles
  ; Handle Alert Level Changes for Possible Speech Bubbles
  (bind _returnvalue (Utilities|Time|SetTimerbyEvent (delegate Custom|TriggerCombatSpeechBubble) (ScalableFloat|GetValueatLevel (Variables|SpeechBubbles|GetSpeechBubbleInCombatInterval)) true))
  (Variables|InternalData|SetHNDL_CombatSpeechBubbleTimer _returnvalue))

(event Custom|UpdateWidgetOffset
  ;[1 1 1] UpdateWidgetOffset
  (if (not (Variables|Default|GetIsWidgetOffsetOverriden))
    (bind _returnvalue (AI|GetNPCWidgetOffset (Fortnite|GetFirstLocalFortPlayerController true) self))
    (Variables|/Script/UMG.Widget|SetRenderTransform_Translation_X (Utilities|Struct|MakeWidgetTransform (Math|Vector2D|MakeVector2D 0.0 _returnvalue)) (UserInterface|GetUserWidgetObject (Variables|Default|GetNPCStatusWidget)))))

(event Custom|TriggerCombatSpeechBubble
  ; Handle Alert Level Changes for Possible Speech Bubbles
  (SpeechBubbles|HandleTriggerSpeechBubbleinCombat))

(event EventOnDied_Event
  (params (DamagedActor :type Actor%20Object%20Reference) (Damage :type Float%20%28single-precision%29) (InstigatedBy :type Controller%20Object%20Reference) (DamageCauser :type Actor%20Object%20Reference) (HitLocation :type Vector :default "0, 0, 0") (FHitComponent :type Primitive%20Component%20Object%20Reference) (BoneName :type Name :default "None") (Momentum :type Vector :default "0, 0, 0"))
  ; Handle Alert Level Changes for Possible Speech Bubbles
  (|CallParentFunction DamagedActor Damage InstigatedBy DamageCauser HitLocation FHitComponent BoneName Momentum)
  (SpeechBubbles|HandleTriggerSpeechBubbleonElim))

(event Custom|TrySnapToGroundOnDeath
  (bind _returnvalue (CallFunction|GetActorLocationwithCapsule))
  (bind (_outhit _returnvalue_2) (Collision|LineTraceByChannel (CallFunction|GetActorLocationwithCapsule) (- (CallFunction|GetActorLocationwithCapsule) "0.000000,0.000000,5000.000000") "TraceTypeQuery1" false (Utilities|Array|MakeArray self) "ForDuration" true "(R=1.000000,G=0.000000,B=0.000000,A=1.000000)" "(R=0.000000,G=1.000000,B=0.000000,A=1.000000)" 2e+01))
  (if _returnvalue_2
    (if (> (Math|Vector|Distance(Vector) _returnvalue (.Location (Collision|BreakHitResult _outhit))) 25.0)
      (Variables|Default|SetDeathFallingOrigin _returnvalue)
      (Variables|Default|SetDeathFallingDestination (Math|Vector|MakeVector (.x (.Location (Collision|BreakHitResult _outhit))) (.y (.Location (Collision|BreakHitResult _outhit))) (+ (.z (.Location (Collision|BreakHitResult _outhit))) (Components|Capsule|GetScaledCapsuleHalfHeight (Variables|Character|GetCapsuleComponent)))))
      (bind _returnvalue_3 (Utilities|Time|SetTimerbyFunctionName 0 "DeathFallPositionInterpolation" 0.001 true true))
      (Variables|Default|SetDeathFallTimerHandle _returnvalue_3))))

(fn DeathFallPositionInterpolation
  (bind (_returnvalue_location _returnvalue_rotation) (Math|Transform|BreakTransform (Transformation|GetActorTransform)))
  (bind _deathfallcurve (Variables|Default|GetDeathFallCurve))
  (bind _deathfallingorigin (Variables|Default|GetDeathFallingOrigin))
  (bind _deathfallingdestination (Variables|Default|GetDeathFallingDestination))
  (bind _output_get (Variables|Default|SetDeathFallTimeCounter (+ (Utilities|Time|GetWorldDeltaSeconds) (Variables|Default|GetDeathFallTimeCounter))))
  (Transformation|Teleport (Math|Vector|Lerp(Vector) _deathfallingorigin _deathfallingdestination (Math|Curves|GetFloatValue _deathfallcurve (* _output_get 2.0))) _returnvalue_rotation)
  (if (Utilities|Operators|Equal(==) (Math|Vector|Lerp(Vector) _deathfallingorigin _deathfallingdestination (Math|Curves|GetFloatValue _deathfallcurve (* _output_get 2.0))) _deathfallingdestination 0)
    (Utilities|Time|ClearandInvalidateTimerbyHandle (Variables|Default|GetDeathFallTimerHandle))))

(fn GetActorLocationWithCapsule
  (bind (_returnvalue_x _returnvalue_y _returnvalue_z) (Math|Vector|BreakVector (Transformation|GetActorLocation self)))
  (return (Math|Vector|MakeVector _returnvalue_x _returnvalue_y (- _returnvalue_z (Components|Capsule|GetScaledCapsuleHalfHeight (Variables|Character|GetCapsuleComponent))))))

(fn HandleEscalationSpeechBubble
  (params (EscalationTags :type Gameplay%20Tag%20Container%20Structure))
  (if (GameplayTags|HasAnyTags EscalationTags (Variables|InternalData|GetEscalation_PatrolBlock_Tag))
    (SpeechBubbles|HandleTriggerSpeechBubbleonEscalationPathBlocked)
    (elif (GameplayTags|HasAnyTags (Variables|GetEscalationTags) (Variables|InternalData|GetEscalation_ADS_Tag))
      (SpeechBubbles|HandleTriggerSpeechBubbleonEscalationADS))))

(fn HandleTriggerSpeechBubbleInCombat
  (bind _speechbubbletags_incombat (Variables|SpeechBubbles|GetSpeechBubbleTags_InCombat))
  (if (Utilities|Array|IsNotEmpty _speechbubbletags_incombat)
    (Class|BPFLAICommonFunctions|PlaySpeechBubble self (Utilities|Array|RandomArrayItem _speechbubbletags_incombat) false true)))

(fn HandleTriggerSpeechBubbleOnAlertLevelChanged
  (params (OldAlertLevel :type EAlertLevel%20Enum :default "Unaware") (NewAlertLevel :type EAlertLevel%20Enum :default "Unaware"))
  (switch Utilities|FlowControl|Switch|SwitchonEAlertLevel NewAlertLevel
    (:Unaware
      (SpeechBubbles|HandleTriggerSpeechBubbleonUnaware))
    (:Suspicious
      (SpeechBubbles|HandleTriggerSpeechBubbleonSuspicious))
    (:LostTarget
      (if (Utilities|Enum|Equal(Enum) OldAlertLevel "Threatened")
        (SpeechBubbles|HandleTriggerSpeechBubbleonLostTargetinCombat)
        (else
          (SpeechBubbles|HandleTriggerSpeechBubbleonLostTarget))))
    (:Threatened
      (SpeechBubbles|HandleTriggerSpeechBubbleonThreatened)
      (CallFunction|HandleCombatSpeechBubbles)))
  (Utilities|Time|ClearandInvalidateTimerbyHandle (Variables|InternalData|GetHNDL_CombatSpeechBubbleTimer)))

(fn HandleTriggerSpeechBubbleOnElim
  (bind _speechbubbletags_onelim (Variables|SpeechBubbles|GetSpeechBubbleTags_OnElim))
  (if (Utilities|Array|IsNotEmpty _speechbubbletags_onelim)
    (Class|BPFLAICommonFunctions|PlaySpeechBubble self (Utilities|Array|RandomArrayItem _speechbubbletags_onelim) true true)))

(fn HandleTriggerSpeechBubbleOnEscalationADS
  (bind _speechbubbletags_escalation_ads (Variables|SpeechBubbles|GetSpeechBubbleTags_Escalation_ADS))
  (if (Utilities|Array|IsNotEmpty _speechbubbletags_escalation_ads)
    (Class|BPFLAICommonFunctions|PlaySpeechBubble self (Utilities|Array|RandomArrayItem _speechbubbletags_escalation_ads) false true)))

(fn HandleTriggerSpeechBubbleOnEscalationPathBlocked
  (bind _speechbubbletags_escalation_pathblocked (Variables|SpeechBubbles|GetSpeechBubbleTags_Escalation_PathBlocked))
  (if (Utilities|Array|IsNotEmpty _speechbubbletags_escalation_pathblocked)
    (Class|BPFLAICommonFunctions|PlaySpeechBubble self (Utilities|Array|RandomArrayItem _speechbubbletags_escalation_pathblocked) false true)))

(fn HandleTriggerSpeechBubbleOnLostTarget
  (bind _speechbubbletags_losttarget (Variables|SpeechBubbles|GetSpeechBubbleTags_LostTarget))
  (if (Utilities|Array|IsNotEmpty _speechbubbletags_losttarget)
    (Class|BPFLAICommonFunctions|PlaySpeechBubble self (Utilities|Array|RandomArrayItem _speechbubbletags_losttarget) false true)))

(fn HandleTriggerSpeechBubbleOnLostTargetInCombat
  (bind _speechbubbletags_losttargetincombat (Variables|SpeechBubbles|GetSpeechBubbleTags_LostTargetInCombat))
  (if (Utilities|Array|IsNotEmpty _speechbubbletags_losttargetincombat)
    (Class|BPFLAICommonFunctions|PlaySpeechBubble self (Utilities|Array|RandomArrayItem _speechbubbletags_losttargetincombat) false true)))

(fn HandleTriggerSpeechBubbleOnSuspicious
  (bind _speechbubbletags_suspicious (Variables|SpeechBubbles|GetSpeechBubbleTags_Suspicious))
  (if (Utilities|Array|IsNotEmpty _speechbubbletags_suspicious)
    (Class|BPFLAICommonFunctions|PlaySpeechBubble self (Utilities|Array|RandomArrayItem _speechbubbletags_suspicious) false true)))

(fn HandleTriggerSpeechBubbleOnThreatened
  (bind _speechbubbletags_threatened (Variables|SpeechBubbles|GetSpeechBubbleTags_Threatened))
  (if (Utilities|Array|IsNotEmpty _speechbubbletags_threatened)
    (Class|BPFLAICommonFunctions|PlaySpeechBubble self (Utilities|Array|RandomArrayItem _speechbubbletags_threatened) false true)))

(fn HandleTriggerSpeechBubbleOnUnaware
  (bind _speechbubbletags_unaware (Variables|SpeechBubbles|GetSpeechBubbleTags_Unaware))
  (if (Utilities|Array|IsNotEmpty _speechbubbletags_unaware)
    (Class|BPFLAICommonFunctions|PlaySpeechBubble self (Utilities|Array|RandomArrayItem _speechbubbletags_unaware) false true)))

(fn OnRep_EnableHealthBar
  (if (Variables|Default|GetEnableHealthBar)
    (CallFunction|NPCEnableHealthBars)))

(fn OnRep_HealthInfo
  (CallFunction|UpdatePawnHealthHUD)
  (Default|CallOnHealthChange))

(fn UpdatePawnHealthHUD
  (bind (_healthpercent_2_7edada384c7908ea98debd9958a249de _shieldpercent_5_b79c63b64aacf275a1621a9315964dcd _maxhealth_6_7d7bef7841b13fb96d2578989cf135bb _maxshield_9_f923e4444e2e476c56fb0e9d9d2f7846) (Utilities|Struct|BreakStructNPCHealthInfo (Variables|Default|GetHealthInfo)))
  (bind _playerstate (Variables|Pawn|GetPlayerState))
  ;[1 1 1] On Pawn Health Changed - Update Health Bar
  (Networking|SwitchHasAuthority
    (:Remote
      (bind _array_element (Utilities|Array|ForEachLoopwithBreak (Fortnite|GetLocalFortPlayerControllers true))
        (:LoopBody
          (if (Math|Boolean|ANDBoolean (< (Class|AbilityFunctions|GetDistanceBetweenActors (Pawn|GetControlledPawn _array_element) self) (Variables|Default|GetNPC%20MaxDistanceFromDamageInstigatorToDrawHealthBar)) (not (Variables|Health|GetbIsDying)) (not (Variables|Default|GetNPC%20StatusWidgetHealthBarDisabled)))
            (bind _npcstatuswidget (Variables|Default|GetNPCStatusWidget)
              (:then
                (bind _asnpcstatus_widget_aiboss (Utilities|Casting|CastToNPCStatusWidget_AIBoss (UserInterface|GetUserWidgetObject _npcstatuswidget)))
                (Class|NPCStatusWidgetAIBoss|EventUpdateStatus _asnpcstatus_widget_aiboss _healthpercent_2_7edada384c7908ea98debd9958a249de _shieldpercent_5_b79c63b64aacf275a1621a9315964dcd _maxhealth_6_7d7bef7841b13fb96d2578989cf135bb _maxshield_9_f923e4444e2e476c56fb0e9d9d2f7846 (Variables|Default|GetShowShields) false)
                ; Set Boss Name (this should be improved by not setting the name every time the widget is updated)
                (bind _aswbp_npcstatus_widget_with_name (Utilities|Casting|CastToWBP_NPCStatusWidget_WithName (UserInterface|GetUserWidgetObject (Variables|Default|GetNPCStatusWidget))))
                (Utilities|IsValid _playerstate
                  (:Is%20Valid
                    (Class|WBPNPCStatusWidgetwithName|SetPlayerName _aswbp_npcstatus_widget_with_name (Utilities|Text|ToText(String) (PlayerState|GetPlayerName _playerstate))))))
              (:else))))))))

(fn ConstructionScript
  (|CallParentFunction))
